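# Physics asset for the LightningDragon backpack's skeletal mesh, defining collision and joint limits for its tail and fins.
# A large root box and spine capsules form the body, with capsules on each tail and fin bone.
# Constraints chain dyn_tail_02 through 09 and hinge eight fins to spine_02 through 06, allowing limited swing and almost no twist.
# The spine bodies, root and dyn_tail_01 have no constraints.
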PhysicsAsset /BRCosmetics/Accessories/FORT_Backpacks/Backpack_F_MED_LightningDragon/Meshes/F_MED_LightningDragon_Pack_Physics_OnEarth
bodies {
 dyn_fin_01 [{
   shapeType=Capsule
   length=4.25
   radius=1.25
   shapeName=dyn_fin_01_capsule
   center=<-1.9573 0.5407 2.0018>
   rotation=<rot 75.0026 0.0013 20.0013>
  }]
 dyn_fin_02 [{
   shapeType=Capsule
   length=2.25
   radius=1.25
   shapeName=dyn_fin_02_capsule
   center=<-0.9217 0.438 1.9899>
   rotation=<rot 60.0007 0.0018 15.0017>
  }]
 dyn_fin_03 [{
   shapeType=Capsule
   length=3.75
   radius=1.25
   shapeName=dyn_fin_03_capsule
   center=<-0.5683 0.6097 1.916>
   rotation=<rot 65.0003 0.0027 15.0027>
  }]
 dyn_fin_04 [{shapeType=Capsule, length=4.25, radius=1.25, shapeName=dyn_fin_04_capsule, center=<-2.217 0 1.7065>, rotation=<rot 69.9999 0.0034 0.0037>}]
 dyn_fin_05 [{
   shapeType=Capsule
   length=3.5
   radius=1.3
   shapeName=dyn_fin_05_capsule
   center=<-1.9383 0.5355 2.6184>
   rotation=<rot 49.9162 1.3486 10.9961>
  }]
 dyn_fin_06 [{shapeType=Capsule, length=3, radius=1.3, shapeName=dyn_fin_06_capsule, center=<-0.7652 0 2.0941>, rotation=<rot 44.9998 0.0003 0.0006>}]
 dyn_fin_07 [{shapeType=Capsule, length=3, radius=1, shapeName=dyn_fin_07_capsule, center=<-0.61 0 1.9253>, rotation=<rot 54.9991 0.0006 0.0014>}]
 dyn_fin_08 [{shapeType=Capsule, length=2.5, radius=1.5, shapeName=dyn_fin_08_capsule, center=<-1.6261 0.2676 2.4675>, rotation=<rot 59.9996 0 10>}]
 dyn_tail_01 [{shapeType=Capsule, length=5, radius=4, shapeName=dyn_tail_01_capsule, center=<3.7905 0 1.7548>, rotation=<rot 90 0 0>}]
 dyn_tail_02 [{shapeType=Capsule, length=5.5, radius=3, shapeName=dyn_tail_02_capsule, center=<2.8041 0 0.5822>, rotation=<rot 90 0 0>}]
 dyn_tail_03 [{shapeType=Capsule, length=4.75, radius=3.75, shapeName=dyn_tail_03_capsule, center=<3.5361 0 1.4144>, rotation=<rot 69.9996 89.9983 90>}]
 dyn_tail_04 [{shapeType=Capsule, length=5, radius=2.75, shapeName=dyn_tail_04_capsule, center=<3.364 0 0.3552>, rotation=<rot 90 129.3522 179.3535>}]
 dyn_tail_05 [{shapeType=Capsule, length=5, radius=2, shapeName=dyn_tail_05_capsule, center=<3.4595 0 0>, rotation=<rot 90 0 0>}]
 dyn_tail_06 [{shapeType=Capsule, length=3.75, radius=2, shapeName=dyn_tail_06_capsule, center=<2.3492 0 0>, rotation=<rot 90 0 0>}]
 dyn_tail_07 [{shapeType=Capsule, length=4.25, radius=2, shapeName=dyn_tail_07_capsule, center=<2.4788 0 0>, rotation=<rot 90 0 0>}]
 dyn_tail_08 [{shapeType=Capsule, length=3.75, radius=2, shapeName=dyn_tail_08_capsule, center=<1.7465 0 0>, rotation=<rot 90 0 0>}]
 dyn_tail_09 [{shapeType=Capsule, length=5, radius=2, shapeName=dyn_tail_09_capsule, center=<1.6419 0.7536 -0.7656>, rotation=<rot 65.0006 -0.0006 10.0003>}]
 root [{shapeType=Box, extentX=28.5, extentY=3, extentZ=31.5, shapeName=root_box, center=<0 2.6054 -29.6622>}]
 spine_01 [{shapeType=Capsule, length=4.5, radius=2.2, shapeName=spine_01_capsule, center=<3.14 0 0>, rotation=<rot 90 0 0>}]
 spine_02 [{shapeType=Capsule, length=5, radius=2.2, shapeName=spine_02_capsule, center=<2.4882 0 0>, rotation=<rot 90 0 0>}]
 spine_03 [{shapeType=Capsule, length=7, radius=2.2, shapeName=spine_03_capsule, center=<4.2825 0 0>, rotation=<rot 90 0 0>}]
 spine_04 [{shapeType=Capsule, length=5, radius=2.2, shapeName=spine_04_capsule, center=<2.7189 0 0>, rotation=<rot 90 0 0>}]
 spine_05 [{shapeType=Capsule, length=5, radius=2.2, shapeName=spine_05_capsule, center=<2.0862 0 0>, rotation=<rot 90 0 0>}]
 spine_06 [{shapeType=Capsule, length=4, radius=2.2, shapeName=spine_06_capsule, center=<2.0768 0 0>, rotation=<rot 90 0 0>}]
 spine_root [{shapeType=Capsule, length=10, radius=4, shapeName=spine_root_capsule, center=<-0.7081 0 -1.7702>, rotation=<rot 84.9996 -179.9906 -174.9901>}]
}
constraints [
 {bone1Name=dyn_tail_09, bone2Name=dyn_tail_08, swing1LimitDegrees=25, swing2LimitDegrees=25, twistLimitDegrees=1}
 {bone1Name=dyn_tail_08, bone2Name=dyn_tail_07, swing1LimitDegrees=25, swing2LimitDegrees=25, twistLimitDegrees=1}
 {bone1Name=dyn_tail_07, bone2Name=dyn_tail_06, swing1LimitDegrees=25, swing2LimitDegrees=25, twistLimitDegrees=1}
 {bone1Name=dyn_tail_06, bone2Name=dyn_tail_05, swing1LimitDegrees=25, swing2LimitDegrees=25, twistLimitDegrees=1}
 {bone1Name=dyn_tail_05, bone2Name=dyn_tail_04, swing1LimitDegrees=15, swing2LimitDegrees=15, twistLimitDegrees=1}
 {bone1Name=dyn_tail_04, bone2Name=dyn_tail_03, swing1LimitDegrees=25, swing2LimitDegrees=25, twistLimitDegrees=1}
 {bone1Name=dyn_tail_03, bone2Name=dyn_tail_02, swing1LimitDegrees=10, swing2LimitDegrees=10, twistLimitDegrees=1}
 {bone1Name=dyn_fin_01, bone2Name=spine_02, swing1LimitDegrees=30, swing2LimitDegrees=65, twistLimitDegrees=1}
 {bone1Name=dyn_fin_02, bone2Name=spine_02, swing1LimitDegrees=30, swing2LimitDegrees=65, twistLimitDegrees=1}
 {bone1Name=dyn_fin_03, bone2Name=spine_03, swing1LimitDegrees=30, swing2LimitDegrees=65, twistLimitDegrees=1}
 {bone1Name=dyn_fin_04, bone2Name=spine_03, swing1LimitDegrees=30, swing2LimitDegrees=65, twistLimitDegrees=1}
 {bone1Name=dyn_fin_05, bone2Name=spine_04, swing1LimitDegrees=30, swing2LimitDegrees=65, twistLimitDegrees=1}
 {bone1Name=dyn_fin_06, bone2Name=spine_04, swing1LimitDegrees=30, swing2LimitDegrees=65, twistLimitDegrees=1}
 {bone1Name=dyn_fin_07, bone2Name=spine_05, swing1LimitDegrees=30, swing2LimitDegrees=40, twistLimitDegrees=1}
 {bone1Name=dyn_fin_08, bone2Name=spine_06, swing1LimitDegrees=30, swing2LimitDegrees=40, twistLimitDegrees=1}
]
