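# Anim Blueprint for the LightningDragon back bling, adding physics motion and elimination reactions to the pack.
# The cached input pose feeds two RigidBody sims in pack-bone space, InAir and OnEarth, blended by bIsSkyDiving.
# An additive Idle/Elim state machine layers on top, with Elim returning automatically.
# A delayed enemy-elimination input event or timed frontend-reactivity loop sets playerElimination, which clears after each evaluation.
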AnimBlueprint /BRCosmetics/Accessories/FORT_Backpacks/Backpack_F_MED_LightningDragon/Meshes/F_MED_LightningDragon_Pack_AnimBP
target_skeleton=&/F_MED_LightningDragon_Pack_Skeleton
graphs [
 AnimGraph {nodes [
   n1 = RigidBody(ComponentPose=Local To Component(LocalPose=Use cached pose 'SavedPose29414'())) {
    bClampLinearTranslationLimitToRefPose=true
    bOverrideWorldGravity=true
    overridePhysicsAsset=&/F_MED_LightningDragon_Pack_Physics_OnEarth
    simulationSpace=BaseBoneSpace
    alphaBoolBlend {blendOption=Linear}
    baseBoneRef {boneName=pack}
    componentAppliedLinearAccClamp=<2000 2000 2000>
    componentLinearAccScale=<0.3 0.3 0.3>
    componentLinearVelScale=<0.3 0.3 0.3>
    overrideWorldGravity {z=-500}
   }
   n2 = Component To Local(ComponentPose=@n1)
   n3 = RigidBody(ComponentPose=Local To Component(LocalPose=Use cached pose 'SavedPose29414'())) {
    bClampLinearTranslationLimitToRefPose=true
    bOverrideWorldGravity=true
    overridePhysicsAsset=&/F_MED_LightningDragon_Pack_Physics_InAir
    simulationSpace=BaseBoneSpace
    alphaBoolBlend {blendOption=Linear}
    baseBoneRef {boneName=pack}
    componentAppliedLinearAccClamp=<2000 2000 2000>
    componentLinearAccScale=<0.4 0.3 0.3>
    componentLinearVelScale=<0.4 0.3 0.3>
    overrideWorldGravity {z=-980}
   }
   n4 = Component To Local(ComponentPose=@n3)
   n5 = Get bIsSkyDiving()
   AnimGraphNode_StateMachine_0 = Idle_State() {is=title AnimGraphNode_StateMachine}
   Output Pose(Result=Apply Additive(Base=Blend Poses by bool(BlendPose_0=@n4, BlendPose_1=@n2, bActiveValue=@n5) {blendType=Linear}, Additive=@AnimGraphNode_StateMachine_0) {alphaBoolBlend {blendOption=Linear}}) {layerGroup=None}
   Save cached pose 'SavedPose29414'(Pose=Input Pose()) {cacheName=SavedPose29414}
  ]}
 Idle_State {
  entry=Idle
  path=AnimGraph.AnimGraphNode_StateMachine_0.Idle_State
  ids=id
  nodes [
   AnimStateNode_0 = Idle()
   AnimStateNode_1 = Elim() {is=title AnimStateNode}
   AnimStateTransitionNode_0 = AnimStateTransitionNode()
   AnimStateTransitionNode_1 = AnimStateTransitionNode() {bAutomaticRuleBasedOnSequencePlayerInState=true, crossfadeDuration=0}
  ]
  links [
   {
    dst=Idle
    rule=AnimGraph.AnimGraphNode_StateMachine_0.Idle_State.AnimStateTransitionNode_1.Transition
    src=Elim
    via=AnimStateTransitionNode_1
   }
   {
    dst=Elim
    rule=AnimGraph.AnimGraphNode_StateMachine_0.Idle_State.AnimStateTransitionNode_0.Transition
    src=Idle
    via=AnimStateTransitionNode_0
   }
  ]
 }
 Idle {
  path=AnimGraph.AnimGraphNode_StateMachine_0.Idle_State.AnimStateNode_0.Idle
  nodes [Output Animation Pose(Result="Sequence Player (additive) 'LightningDragon_Idle_Pack'"() {sequence=/BRCosmetics/Accessories/FORT_Backpacks/Backpack_F_MED_LightningDragon/Animation/LightningDragon_Idle_Pack})]
 }
 Elim {
  path=AnimGraph.AnimGraphNode_StateMachine_0.Idle_State.AnimStateNode_1.Elim
  nodes [
   n1 = "Sequence Player (additive) 'LightningDragon_Elim_Pack'"() {bLoopAnimation=false, sequence=/BRCosmetics/Accessories/FORT_Backpacks/Backpack_F_MED_LightningDragon/Animation/LightningDragon_Elim_Pack}
   Output Animation Pose(Result=@n1)
  ]
 }
 Transition {
  path=AnimGraph.AnimGraphNode_StateMachine_0.Idle_State.AnimStateTransitionNode_0.Transition
  nodes [Result(bCanEnterTransition=Get playerElimination() -> (playerElimination:bool))]
 }
 Transition {path=AnimGraph.AnimGraphNode_StateMachine_0.Idle_State.AnimStateTransitionNode_1.Transition, nodes [Result()]}
 EventGraph {nodes [
   n1 = Event BlueprintProcessAnimInputEvent()
   > Branch(Condition=IsNamed(self=@n1.AnimInputEvent, InEventName=aie.EnemyPlayerEliminated))
   > Delay(Duration=0.5)
   > Set playerElimination(playerElimination:bool=true) -> (Output_Get:bool)
   Event BlueprintPostEvaluateAnimation() {is=title K2Node_Event}
   > Set playerElimination(playerElimination=false)
   Event BlueprintUpdateAnimation()
   > Branch(Condition=Get bShouldDisplayFrontendReactivity() -> (bShouldDisplayFrontendReactivity:bool))
   > n2 = Sequence()
   > Do Once()
   > n3 = Delay(Duration=0.6)
   Set playerElimination(execute=[@n4.then, @n3.then], playerElimination=true)
   n4 = Delay(execute=@n2.then_1, Duration=6.012)
   Reroute Node(InputPin:Wildcard) -> (OutputPin:Wildcard)
  ]}
]
variables [
 DkyDiving {type=number}
 notSkyDiving {type=number}
 playerElimination {type=boolean}
]
